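; Performance test Blueprint for UtopiaCity grass materials, cycling a plane through shading variants.
; On BeginPlay, timed delays step slot 0 of the Plane through the PDO, Parallax and MultiSample material instances in sequence.
; The sequence runs once without looping; overlap, tick and construction script are empty, and no measurement logic is present.
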(event EventBeginPlay
  (Utilities|FlowControl|Delay 1e+01)
  (Rendering|Material|SetMaterial (Variables|Default|GetPlane) 0 "/Hermes/Sets/UtopiaCity/Materials/Grass/Test/MI_UtopiaCity_Grass_Test_PDO.MI_UtopiaCity_Grass_Test_PDO")
  (Utilities|FlowControl|Delay 1e+01)
  (Rendering|Material|SetMaterial (Variables|Default|GetPlane) 0 "/Hermes/Sets/UtopiaCity/Materials/Grass/Test/MI_UtopiaCity_Grass_Test_Parallax.MI_UtopiaCity_Grass_Test_Parallax")
  (Utilities|FlowControl|Delay 1e+01)
  (Rendering|Material|SetMaterial (Variables|Default|GetPlane) 0 "/Hermes/Sets/UtopiaCity/Materials/Grass/Test/MI_UtopiaCity_Grass_Test_MultiSample.MI_UtopiaCity_Grass_Test_MultiSample"))

(event Collision|EventActorBeginOverlap
  (params (OtherActor :type Actor%20Object%20Reference)))

(event EventTick
  (params (DeltaSeconds :type Float%20%28single-precision%29)))

(fn ConstructionScript)
